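; Parent spawner actor that places a buried chest prop on the server once playlist data is ready, then removes itself after a short randomized lifespan.
; Spawning requires two data-driven enable flags and a weighted roll between min and max chance.
; Placement uses either a line trace directly under the spawner or an EQS query fed an overridable radius, with random yaw.
; The construction script debug-draws the trace line and spawn radius.
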(event EventBeginPlay
  (if (Networking|IsServer)
    (Class|FortAsyncActionCurrentPlaylistReady|PlaylistDataReadyAsync
      (:OnReady
        (if (and (Class|AthenaFunctionLibrary|RowtoBool (Variables|BuriedChest|GetRow_SpawnEnabled)) (Class|AthenaFunctionLibrary|RowtoBool (Variables|BuriedChest|GetRow_SpawnGlobalEnabled)))
          ;[1 1 1] Check Spawn Rate and do a roll. If pass spawn chest else kill the actor
          (if (CallFunction|CheckforSpawnRate)
            (if (Variables|BuriedChest|GetSpawnDirectlyUnderSpawner)
              ;[1 1 1] If spawn directly under, do a trace and check if we hit. Spawn the chest if hit. Destroy spawner after in either case.
              (bind (_outhit _returnvalue) (Collision|LineTraceByChannel (Transformation|GetActorLocation) (Math|Vector|vector+vector (Transformation|GetActorLocation) (Math|Vector|MakeVector 0.0 0.0 (Variables|BuriedChest|GetDirectlyUnderTraceDownDistance))) (Variables|BuriedChest|GetDirectlyUnderTraceChannel)))
              (if _returnvalue
                (CallFunction|SpawnChest (Variables|BuriedChest|GetBuriedChestPropClass) (Math|Transform|MakeTransform (.Location (Collision|BreakHitResult _outhit)) (Math|Rotator|MakeRotator 0.0 0.0 (Math|Random|RandomFloatinRange 0.0 3.6e+02)) "1.000000,1.000000,1.000000"))
                (else
                  _))
              (Actor|SetLifeSpan (Math|Random|RandomFloatinRange 0.1 0.5))
              (else
                ;[1 1 1] Run EQS and bind to it finishing
                (bind _returnvalue_2 (AI|EQS|RunEQSQuery (Variables|BuriedChest|GetEQS) self "AllMatching"))
                (AI|EQS|SetNamedParam _returnvalue_2 (Variables|BuriedChest|GetSpawnRadiusParamName) (ScalableFloat|GetValueatLevel (Variables|BuriedChest|GetRow_SpawnRadius))) ; Pass in the outer spawn radius as a param so it can be overriden on a per spawner basis
                (AI|EQS|BindEventtoOnQueryFinished _returnvalue_2 (delegate Custom|EQSFinished))))
            (else
              _))
          (else
            _))
        (Actor|SetLifeSpan (Math|Random|RandomFloatinRange 0.01 0.2))))))

(event Custom|EQSFinished
  (params (QueryInstance :type EQS%20Query%20Instance%20Object%20Reference) (QueryStatus :type EEnvQueryStatus%20Enum :default "Processing"))
  ;[1 1 1] Pull Value From EQS when finished and spawn the prop that spawns the chest. After, kill the actor
  (bind _resultlocations (AI|EQS|GetQueryResultsasLocations QueryInstance))
  (if (> (Utilities|Array|Length _resultlocations) 0)
    (CallFunction|SpawnChest (Variables|BuriedChest|GetBuriedChestPropClass) (Math|Transform|MakeTransform (Utilities|Array|Get(aref) _resultlocations) (Math|Rotator|MakeRotator 0.0 0.0 (Math|Random|RandomFloatinRange 0.0 3.6e+02)) "1.000000,1.000000,1.000000"))
    (else
      _))
  (Actor|SetLifeSpan (Math|Random|RandomFloatinRange 0.1 0.5)))

(fn CheckForSpawnRate
  (return (Math|Random|RandomBoolwithWeight (Math|Random|RandomFloatinRange (* (ScalableFloat|GetValueatLevel (Variables|BuriedChest|GetRow_MinChanceToSpawn)) 0.01) (* (ScalableFloat|GetValueatLevel (Variables|BuriedChest|GetRow_MaxChanceToSpawn)) 0.01)))))

(fn SpawnChest
  (params (Class :type Actor%20Class%20Reference) (SpawnTransform :type Transform))
  (Game|SpawnActorfromClass Class SpawnTransform "Undefined" "SelectDefaultAtRuntime" 0 0))

(fn ConstructionScript
  (bind _debugdrawdurationtime (Variables|BuriedChest|GetDebugDrawDurationTime))
  ;[1 1 1] Draw Debug to visualize placement
  (Rendering|Debug|DrawDebugLine (Transformation|GetActorLocation) (Math|Vector|vector+vector (Transformation|GetActorLocation) (Math|Vector|MakeVector 0.0 0.0 (Variables|BuriedChest|GetDirectlyUnderTraceDownDistance))) "(R=1.000000,G=0.000000,B=0.013448,A=1.000000)" _debugdrawdurationtime)
  (Rendering|Debug|DrawDebugCircle (Transformation|GetActorLocation) (ScalableFloat|GetValueatLevel (Variables|BuriedChest|GetRow_SpawnRadius)) 32 "(R=1.000000,G=0.000000,B=0.045438,A=1.000000)" _debugdrawdurationtime 0.0 "0.000000,1.000000,0.000000" "1.000000,0.000000,0.000000"))
